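# Physics asset for the right-arm mesh of the male medium Anarchy Acres Farmer player body, setting up physics for a pouch on the arm.
# A capsule on upperarm_r and a box on dyn_pouch supply the main shapes, linked through a tiny capsule on dyn_mstr_pouch.
# The master pouch bone rotates freely relative to the upper arm, while the pouch has narrow swing and twist limits against that master.
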PhysicsAsset /BRCosmetics/Characters/Player/Male/Medium/Bodies/M_MED_Anarchy_Acres_Farmer/Meshes/M_MED_Anarchy_Acres_Farmer_ArmRight_Physics
bodies {
 dyn_mstr_pouch [{shapeType=Capsule, length=0.505, radius=0.505}]
 dyn_pouch [{shapeType=Box, extentX=4.6149, extentY=6.4393, extentZ=9.6579, center=<1.8797 -0.2089 -1.009>}]
 upperarm_r [{shapeType=Capsule, length=8.6135, radius=7.0792, center=<-13.4091 0.6782 2.6386>, rotation=<rot 90 0 0>}]
}
constraints [
 {bone1Name=dyn_pouch, bone2Name=dyn_mstr_pouch, swing1LimitDegrees=5, swing2LimitDegrees=1, twistLimitDegrees=1}
 {
  bone1Name=dyn_mstr_pouch
  bone2Name=upperarm_r
  swing1LimitDegrees=45
  swing1Motion=Free
  swing2LimitDegrees=45
  swing2Motion=Free
  twistLimitDegrees=45
  twistMotion=Free
 }
]
