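# Physics asset for the back plate on the male medium Anarchy Acres Farmer body, set up for its dynamic plate bones.
# Two capsules on spine_04 form the torso volume, alongside a tiny box on dyn_mstr_backPlate and a large offset box for dyn_backPlate.
# The master bone joins the spine with free swing and twist.
# The plate hinges from the master with one wide swing limit and tight limits on the other swing and twist.
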PhysicsAsset /BRCosmetics/Characters/Player/Male/Medium/Bodies/M_MED_Anarchy_Acres_Farmer/Meshes/M_MED_Anarchy_Acres_Farmer_BackPlate_Physics
bodies {
 dyn_backPlate [{shapeType=Box, extentX=30.0704, extentY=7.5911, extentZ=21.6029, center=<13.5475 1.2854 0>}]
 dyn_mstr_backPlate [{shapeType=Box, extentX=1.01, extentY=1.01, extentZ=1.01}]
 spine_04 [{shapeType=Capsule, length=20.87, radius=17.2497, center=<-7.77 8.68 0>}, {shapeType=Capsule, length=15, radius=9, center=<10.44 2.2 0>}]
}
constraints [
 {
  bone1Name=dyn_mstr_backPlate
  bone2Name=spine_04
  swing1LimitDegrees=45
  swing1Motion=Free
  swing2LimitDegrees=45
  swing2Motion=Free
  twistLimitDegrees=45
  twistMotion=Free
 }
 {bone1Name=dyn_backPlate, bone2Name=dyn_mstr_backPlate, swing1LimitDegrees=45, swing2LimitDegrees=10, twistLimitDegrees=2}
]
