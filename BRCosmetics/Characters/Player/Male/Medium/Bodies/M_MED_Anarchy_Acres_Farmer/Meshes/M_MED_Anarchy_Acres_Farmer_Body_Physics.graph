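# Physics asset for the Anarchy Acres Farmer male medium body cosmetic, covering dangling front and back belts, a rear flap and a flashlight.
# Each piece chains from a small master body that is free in swing and twist on the two-capsule pelvis.
# Box segments form the belts with tight angular limits after a free first link.
# The flap hinges on one swing axis, while the flashlight capsule swings wider with restricted twist.
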PhysicsAsset /BRCosmetics/Characters/Player/Male/Medium/Bodies/M_MED_Anarchy_Acres_Farmer/Meshes/M_MED_Anarchy_Acres_Farmer_Body_Physics
bodies {
 dyn_beltBack1 [{shapeType=Capsule, length=0, radius=0.5, center=<0.4553 0.1491 0.1401>}]
 dyn_beltBack2 [{shapeType=Box, extentX=2.9611, extentY=0.8274, extentZ=3.5779, center=<1.4038 -0.125 0.017>}]
 dyn_beltBack3 [{shapeType=Box, extentX=2.6842, extentY=0.7501, extentZ=3.6457, center=<1.359 -0.0077 0.0243>}]
 dyn_beltBack4 [{shapeType=Box, extentX=4.0261, extentY=0.6258, extentZ=3.5698, center=<1.9226 0.0149 0.0118>}]
 dyn_beltBack5 [{shapeType=Box, extentX=2.0714, extentY=0.9139, extentZ=3.1167, center=<0.9593 -0.1741 0.1541>, rotation=<rot 7.0053 0 0>}]
 dyn_beltFront1 [{shapeType=Capsule, length=0, radius=0.5, center=<0.3968 0.0752 -0.0234>}]
 dyn_beltFront2 [{shapeType=Box, extentX=4.1863, extentY=1.213, extentZ=5.0169, center=<2.1099 -0.0071 -0.0426>}]
 dyn_beltFront3 [{shapeType=Box, extentX=3.7992, extentY=0.8584, extentZ=4.977, center=<1.9155 -0.0157 -0.0223>}]
 dyn_beltFront4 [{shapeType=Box, extentX=2.3382, extentY=0.9272, extentZ=4.4988, center=<1.1829 0.0597 -0.0219>}]
 dyn_buttFlap1 [{shapeType=Capsule, length=8.5125, radius=1.2293, center=<-0.3891 0.0196 0>}]
 dyn_buttFlap2 [{shapeType=Box, extentX=9, extentY=5, extentZ=21, center=<4.0815 -0.2596 0.0001>, rotation=<rot 0 -11.4939 0>}]
 dyn_flashlight [{shapeType=Capsule, length=9.2498, radius=5.1291, center=<4.0797 -0.2462 0.0036>, rotation=<rot 0 -88.6562 -89.9763>}]
 dyn_mstr_beltBack [{shapeType=Capsule, length=0, radius=0.5}]
 dyn_mstr_beltFront [{shapeType=Capsule, length=0, radius=0.5}]
 dyn_mstr_buttFlap [{shapeType=Capsule, length=0.505, radius=0.505}]
 dyn_mstr_flashlight [{shapeType=Capsule, length=0.505, radius=0.505}]
 pelvis [
  {shapeType=Capsule, length=19.683, radius=9.1237, center=<-6.0972 -0.8992 -0.1469>}
  {shapeType=Capsule, length=5.25, radius=5, center=<4.6716 -8.8009 -2.8428>}
 ]
}
constraints [
 {bone1Name=dyn_flashlight, bone2Name=dyn_mstr_flashlight, swing1LimitDegrees=45, swing2LimitDegrees=45, twistLimitDegrees=15}
 {bone1Name=dyn_buttFlap1, bone2Name=dyn_mstr_buttFlap, swing1LimitDegrees=45, swing2LimitDegrees=0, twistLimitDegrees=0}
 {bone1Name=dyn_buttFlap2, bone2Name=dyn_buttFlap1, swing1LimitDegrees=45, swing2LimitDegrees=0, twistLimitDegrees=0}
 {
  bone1Name=dyn_mstr_flashlight
  bone2Name=pelvis
  swing1LimitDegrees=45
  swing1Motion=Free
  swing2LimitDegrees=45
  swing2Motion=Free
  twistLimitDegrees=45
  twistMotion=Free
 }
 {
  bone1Name=dyn_mstr_buttFlap
  bone2Name=pelvis
  swing1LimitDegrees=45
  swing1Motion=Free
  swing2LimitDegrees=45
  swing2Motion=Free
  twistLimitDegrees=45
  twistMotion=Free
 }
 {
  bone1Name=dyn_mstr_beltFront
  bone2Name=pelvis
  swing1LimitDegrees=45
  swing1Motion=Free
  swing2LimitDegrees=45
  swing2Motion=Free
  twistLimitDegrees=45
  twistMotion=Free
 }
 {
  bone1Name=dyn_beltFront1
  bone2Name=dyn_mstr_beltFront
  swing1LimitDegrees=45
  swing1Motion=Free
  swing2LimitDegrees=45
  swing2Motion=Free
  twistLimitDegrees=45
  twistMotion=Free
 }
 {bone1Name=dyn_beltFront2, bone2Name=dyn_beltFront1, swing1LimitDegrees=20, swing2LimitDegrees=20, twistLimitDegrees=5}
 {bone1Name=dyn_beltFront3, bone2Name=dyn_beltFront2, swing1LimitDegrees=20, swing2LimitDegrees=20, twistLimitDegrees=5}
 {bone1Name=dyn_beltFront4, bone2Name=dyn_beltFront3, swing1LimitDegrees=20, swing2LimitDegrees=20, twistLimitDegrees=5}
 {
  bone1Name=dyn_mstr_beltBack
  bone2Name=pelvis
  swing1LimitDegrees=45
  swing1Motion=Free
  swing2LimitDegrees=45
  swing2Motion=Free
  twistLimitDegrees=45
  twistMotion=Free
 }
 {
  bone1Name=dyn_beltBack1
  bone2Name=dyn_mstr_beltBack
  swing1LimitDegrees=45
  swing1Motion=Free
  swing2LimitDegrees=45
  swing2Motion=Free
  twistLimitDegrees=45
  twistMotion=Free
 }
 {bone1Name=dyn_beltBack2, bone2Name=dyn_beltBack1, swing1LimitDegrees=20, swing2LimitDegrees=10, twistLimitDegrees=5}
 {bone1Name=dyn_beltBack3, bone2Name=dyn_beltBack2, swing1LimitDegrees=20, swing2LimitDegrees=10, twistLimitDegrees=5}
 {bone1Name=dyn_beltBack4, bone2Name=dyn_beltBack3, swing1LimitDegrees=20, swing2LimitDegrees=10, twistLimitDegrees=5}
 {bone1Name=dyn_beltBack5, bone2Name=dyn_beltBack4, swing1LimitDegrees=20, swing2LimitDegrees=10, twistLimitDegrees=5}
]
