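# Physics asset for the right-leg belt piece on the medium male Anarchy Acres Farmer player cosmetic, covering its dynamic bone chain.
# Small capsules sit on the master bone and first link, with boxes on the three lower links.
# Constraints join the bones in sequence: the first link swings and twists freely from the master.
# Each later link has tight swing and even tighter twist limits.
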PhysicsAsset /BRCosmetics/Characters/Player/Male/Medium/Bodies/M_MED_Anarchy_Acres_Farmer/Meshes/M_MED_Anarchy_Acres_Farmer_LegRight_Physics
bodies {
 dyn_beltLeg1 [{shapeType=Capsule, length=0, radius=0.5, center=<0.7886 0.0375 0.0339>}]
 dyn_beltLeg2 [{shapeType=Box, extentX=2.5041, extentY=1.4468, extentZ=4.4565, center=<1.3127 -0.0056 -0.0153>, rotation=<rot -1.135 0 0>}]
 dyn_beltLeg3 [{shapeType=Box, extentX=1.9243, extentY=1.2918, extentZ=4.2181, center=<0.971 -0.0044 -0.0023>}]
 dyn_beltLeg4 [{shapeType=Box, extentX=2.3758, extentY=1.1406, extentZ=3.4417, center=<0.8658 0.1354 -0.016>}]
 dyn_mstr_beltLeg [{shapeType=Capsule, length=0, radius=0.5}]
}
constraints [
 {
  bone1Name=dyn_beltLeg1
  bone2Name=dyn_mstr_beltLeg
  swing1LimitDegrees=45
  swing1Motion=Free
  swing2LimitDegrees=45
  swing2Motion=Free
  twistLimitDegrees=45
  twistMotion=Free
 }
 {bone1Name=dyn_beltLeg2, bone2Name=dyn_beltLeg1, swing1LimitDegrees=20, swing2LimitDegrees=20, twistLimitDegrees=5}
 {bone1Name=dyn_beltLeg3, bone2Name=dyn_beltLeg2, swing1LimitDegrees=20, swing2LimitDegrees=20, twistLimitDegrees=5}
 {bone1Name=dyn_beltLeg4, bone2Name=dyn_beltLeg3, swing1LimitDegrees=20, swing2LimitDegrees=20, twistLimitDegrees=5}
]
