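# Animation Blueprint for the Anarchy Acres Farmer body cosmetic, adding secondary motion to a pose copied from the attached parent mesh.
# Weighted constraints position pouch and belt bones, then chained RigidBody nodes simulate body, right arm, right leg and back plate physics.
# Back plate physics switches off with backbling; without backbling, a fixed offset applies when rigid bodies are disabled or LOD exceeds 1.
# Event-updated crouch and skydive alphas blend additive piston and back plate offsets.
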AnimBlueprint /BRCosmetics/Characters/Player/Male/Medium/Bodies/M_MED_Anarchy_Acres_Farmer/Meshes/M_MED_Anarchy_Acres_Farmer_AnimBP
target_skeleton=&/M_MED_Anarchy_Acres_Farmer_Skeleton
graphs [
 AnimGraph {nodes [
   n1 = "Constraint - dyn_mstr_pouch (PPPP)"!(
    ComponentPose=Local To Component(LocalPose=Copy Pose From Mesh() {bUseAttachedParent=true})
    Alpha=1
    ConstraintWeights_0=0.0337
    ConstraintWeights_1:float=0.3876
    ConstraintWeights_2:float=0.2748
    ConstraintWeights_3:float=0.304
   ) -> (Pose) {
    alphaBoolBlend {blendOption=Linear}
    boneToModify {boneName=dyn_mstr_pouch}
    constraintSetup [
     {targetBone {boneName=clavicle_r}}
     {targetBone {boneName=upperarm_twist_01_r}}
     {targetBone {boneName=upperarm_r}}
     {targetBone {boneName=upperarm_twist_02_r}}
    ]
    constraintWeights [1,1,1,1]
   }
   n2 = "Constraint - dyn_mstr_beltLeg (PPP)"!(ComponentPose=@n1, Alpha=1, ConstraintWeights_0=0.0466, ConstraintWeights_1:float=0.0076, ConstraintWeights_2:float=0.9458) -> (Pose) {
    alphaBoolBlend {blendOption=Linear}
    boneToModify {boneName=dyn_mstr_beltLeg}
    constraintSetup [{targetBone {boneName=thigh_r}}, {targetBone {boneName=pelvis}}, {targetBone {boneName=thigh_twist_01_r}}]
    constraintWeights [1,1,1]
   }
   n3 = "Constraint - dyn_mstr_beltFront (PPPP)"!(
    ComponentPose=@n2
    Alpha=1
    ConstraintWeights_0=0.4888
    ConstraintWeights_1:float=0.0863
    ConstraintWeights_2:float=0.049
    ConstraintWeights_3:float=0.376
   ) -> (Pose) {
    alphaBoolBlend {blendOption=Linear}
    boneToModify {boneName=dyn_mstr_beltFront}
    constraintSetup [{targetBone {boneName=pelvis}}, {targetBone {boneName=spine_02}}, {targetBone {boneName=spine_03}}, {targetBone {boneName=spine_01}}]
    constraintWeights [1,1,1,1]
   }
   n4 = "Constraint - dyn_mstr_beltBack (PPPPP)"!(
    ComponentPose=@n3
    Alpha=1
    ConstraintWeights_0=0.7869
    ConstraintWeights_1:float=0.0064
    ConstraintWeights_2:float=0.0041
    ConstraintWeights_3:float=0.0095
    ConstraintWeights_4:float=0.1931
   ) -> (Pose) {
    alphaBoolBlend {blendOption=Linear}
    boneToModify {boneName=dyn_mstr_beltBack}
    constraintSetup [
     {targetBone {boneName=pelvis}}
     {targetBone {boneName=spine_02}}
     {targetBone {boneName=spine_03}}
     {targetBone {boneName=thigh_twist_01_r}}
     {targetBone {boneName=spine_01}}
    ]
    constraintWeights [1,1,1,1,1]
   }
   n5 = RigidBody(ComponentPose=@n4) {
    bClampLinearTranslationLimitToRefPose=true
    bOverrideWorldGravity=true
    overridePhysicsAsset=&/M_MED_Anarchy_Acres_Farmer_Body_Physics
    simulationSpace=BaseBoneSpace
    alphaBoolBlend {blendOption=Linear}
    baseBoneRef {boneName=pelvis}
    componentAppliedLinearAccClamp=<3000 3000 3000>
    componentLinearAccScale=<0.4 0.3 0.3>
    componentLinearVelScale=<0.4 0.3 0.3>
    overrideWorldGravity {z=-980}
   }
   n6 = RigidBody(ComponentPose=@n5) {
    bClampLinearTranslationLimitToRefPose=true
    bOverrideWorldGravity=true
    overridePhysicsAsset=&/M_MED_Anarchy_Acres_Farmer_ArmRight_Physics
    simulationSpace=BaseBoneSpace
    alphaBoolBlend {blendOption=Linear}
    baseBoneRef {boneName=upperarm_r}
    componentAppliedLinearAccClamp=<3000 3000 3000>
    componentLinearAccScale=<0.4 0.3 0.3>
    componentLinearVelScale=<0.4 0.3 0.3>
    overrideWorldGravity {z=-980}
   }
   n7 = RigidBody(ComponentPose=@n6) {
    bClampLinearTranslationLimitToRefPose=true
    bOverrideWorldGravity=true
    overridePhysicsAsset=&/M_MED_Anarchy_Acres_Farmer_LegRight_Physics
    simulationSpace=BaseBoneSpace
    alphaBoolBlend {blendOption=Linear}
    baseBoneRef {boneName=thigh_r}
    componentAppliedLinearAccClamp=<3000 3000 3000>
    componentLinearAccScale=<0.4 0.3 0.3>
    componentLinearVelScale=<0.4 0.3 0.3>
    overrideWorldGravity {z=-980}
   }
   n8 = NOT Boolean(A=Get isWearingBackbling())
   n9 = RigidBody!(ComponentPose=@n7, bAlphaBoolEnabled:bool=@n8, ExternalForce) -> (Pose) {
    alphaInputType=Bool
    bClampLinearTranslationLimitToRefPose=true
    bOverrideWorldGravity=true
    overridePhysicsAsset=&/M_MED_Anarchy_Acres_Farmer_BackPlate_Physics
    simulationSpace=BaseBoneSpace
    alphaBoolBlend {blendOption=Linear}
    baseBoneRef {boneName=spine_05}
    componentAppliedLinearAccClamp=<3000 3000 3000>
    componentLinearAccScale=<0.4 0.3 0.3>
    componentLinearVelScale=<0.4 0.3 0.3>
    overrideWorldGravity {z=-980}
   }
   n10 = Get crouching() -> (crouching:double)
   n11 = "Transform (Modify) Bone - Bone: piston_l"(ComponentPose=@n9, Alpha=@n10, Translation="0,0,-5", Rotation="0,0,0", Scale="1,1,1") {
    translationMode=BMM_Additive
    alphaBoolBlend {blendOption=Linear}
    alphaScaleBiasClamp {bInterpResult=true, interpSpeedDecreasing=7.5, interpSpeedIncreasing=7.5}
    boneToModify {boneName=piston_l}
    translation {z=-5}
   }
   n12 = "Transform (Modify) Bone - Bone: piston_r"(ComponentPose=@n11, Alpha=@n10, Translation="0,-5,0", Rotation="0,0,0", Scale="1,1,1") {
    translationMode=BMM_Additive
    alphaBoolBlend {blendOption=Linear}
    alphaScaleBiasClamp {bInterpResult=true, interpSpeedDecreasing=7.5, interpSpeedIncreasing=7.5}
    boneToModify {boneName=piston_r}
    translation {y=-5}
   }
   n13 = "Transform (Modify) Bone - Bone: dyn_mstr_backPlate"!(
    ComponentPose=@n12
    bAlphaBoolEnabled:bool=AND Boolean(A=Get isRigidBodyDisabled() -> (isRigidBodyDisabled:bool), B=@n8)
    Translation="0,-1,0"
    Rotation="0,0,5"
    Scale="1,1,1"
   ) -> (Pose) {
    alphaInputType=Bool
    rotationMode=BMM_Additive
    translationMode=BMM_Additive
    alphaBoolBlend {blendOption=Linear}
    boneToModify {boneName=dyn_mstr_backPlate}
    rotation {roll=5}
    translation {y=-1}
   }
   n14 = "Transform (Modify) Bone - Bone: dyn_mstr_backPlate"!(ComponentPose=@n13, Alpha:float=Get isSkydiving() -> (isSkydiving:double), Translation="-0.000,2.766,0.000", Rotation="0,0,5", Scale="1,1,1") -> (Pose) {
    rotationMode=BMM_Additive
    translationMode=BMM_Additive
    alphaBoolBlend {blendOption=Linear}
    alphaScaleBiasClamp {bInterpResult=true}
    boneToModify {boneName=dyn_mstr_backPlate}
    rotation {roll=5}
    translation {y=2.766}
   }
   Output Pose(Result=Component To Local(ComponentPose=@n14)) {layerGroup=None}
  ]}
 EventGraph {nodes [
   Event BlueprintUpdateAnimation()
   > Set crouching(crouching:double="To Float (Boolean)"(InBool=Get bIsCrouching())) -> (Output_Get:double)
   n1 = Get CurrentLOD()
   n2 = Get bRigidBodyEnabled()
   > Set isRigidBodyDisabled(isRigidBodyDisabled:bool=OR Boolean(A=NOT Boolean(A=@n2), B="integer > integer"(A=@n1, B=1) {is=title K2Node_CallFunction})) -> (Output_Get:bool)
   > Set isSkydiving(isSkydiving:double="To Float (Boolean)"(InBool=Get bIsSkydiveFloating())) -> (Output_Get:double)
   Event OnCharacterCustomizationLoaded()
   > Do Once()
   n3 = CheckForBackbling()
   > Set isWearingBackbling(isWearingBackbling=OR Boolean(A=@n3, B=@n3.bIsBackblingLong))
  ]}
]
variables [
 crouching {type=number}
 isRigidBodyDisabled {type=boolean}
 isSkydiving {type=number}
 isWearingBackbling {type=boolean}
]
