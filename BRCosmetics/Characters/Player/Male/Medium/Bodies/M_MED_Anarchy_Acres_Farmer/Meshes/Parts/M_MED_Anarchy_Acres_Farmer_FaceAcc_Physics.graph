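# Physics asset for the Anarchy Acres Farmer male face-accessory mesh, rigging a hat and a wheat straw for secondary motion.
# Four two-segment box chains (front, back, left, right) branch from a small hat master capsule with tight swing and twist limits.
# A five-segment box straw chain hangs from a wheat master capsule, allowing wider bend on one swing axis than the other, with little twist.
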PhysicsAsset /BRCosmetics/Characters/Player/Male/Medium/Bodies/M_MED_Anarchy_Acres_Farmer/Meshes/Parts/M_MED_Anarchy_Acres_Farmer_FaceAcc_Physics
bodies {
 dyn_hatBack1 [{shapeType=Box, extentX=3.5, extentY=12, extentZ=1, center=<1.75 0 0>}]
 dyn_hatBack2 [{shapeType=Box, extentX=4, extentY=15, extentZ=1, center=<2 0 0>}]
 dyn_hatFront1 [{shapeType=Box, extentX=3.5, extentY=12, extentZ=1, center=<1.75 0 0>}]
 dyn_hatFront2 [{shapeType=Box, extentX=4, extentY=15, extentZ=1, center=<2 0 0>}]
 dyn_hatLeft1 [{shapeType=Box, extentX=3.5, extentY=12, extentZ=1, center=<1.75 0 0>}]
 dyn_hatLeft2 [{shapeType=Box, extentX=4, extentY=15, extentZ=1, center=<2 0 0>}]
 dyn_hatRight1 [{shapeType=Box, extentX=3.5, extentY=12, extentZ=1, center=<1.75 0 0>, rotation=<rot 0 -15.738 0>}]
 dyn_hatRight2 [{shapeType=Box, extentX=4, extentY=15, extentZ=1, center=<2 0 0>, rotation=<rot 0 -15.8598 0>}]
 dyn_mstr_hat [{shapeType=Capsule, length=0.505, radius=0.505}]
 dyn_mstr_wheat [{shapeType=Capsule, length=0.505, radius=0.505}]
 dyn_straw1 [{shapeType=Box, extentX=3.01, extentY=1.01, extentZ=1.01, center=<2.4649 -0.0087 0.0016>}]
 dyn_straw2 [{shapeType=Box, extentX=3.51, extentY=1.01, extentZ=1.01, center=<1.7054 0.0021 0.006>}]
 dyn_straw3 [{shapeType=Box, extentX=2.51, extentY=1.01, extentZ=1.01, center=<1.4211 0.0006 0.0108>}]
 dyn_straw4 [{shapeType=Box, extentX=2.51, extentY=1.01, extentZ=1.01, center=<1.323 0.0016 0.0107>}]
 dyn_straw5 [{shapeType=Box, extentX=6.428, extentY=1.1588, extentZ=1.4614, center=<3.1766 -0.1074 -0.2481>}]
}
constraints [
 {bone1Name=dyn_straw1, bone2Name=dyn_mstr_wheat, swing1LimitDegrees=15, swing2LimitDegrees=2, twistLimitDegrees=2}
 {bone1Name=dyn_straw2, bone2Name=dyn_straw1, swing1LimitDegrees=15, swing2LimitDegrees=2, twistLimitDegrees=2}
 {bone1Name=dyn_straw3, bone2Name=dyn_straw2, swing1LimitDegrees=15, swing2LimitDegrees=2, twistLimitDegrees=2}
 {bone1Name=dyn_straw4, bone2Name=dyn_straw3, swing1LimitDegrees=15, swing2LimitDegrees=2, twistLimitDegrees=2}
 {bone1Name=dyn_straw5, bone2Name=dyn_straw4, swing1LimitDegrees=15, swing2LimitDegrees=2, twistLimitDegrees=2}
 {bone1Name=dyn_hatFront1, bone2Name=dyn_mstr_hat, swing1LimitDegrees=5, swing2LimitDegrees=5, twistLimitDegrees=2}
 {bone1Name=dyn_hatFront2, bone2Name=dyn_hatFront1, swing1LimitDegrees=5, swing2LimitDegrees=5, twistLimitDegrees=2}
 {bone1Name=dyn_hatLeft1, bone2Name=dyn_mstr_hat, swing1LimitDegrees=5, swing2LimitDegrees=5, twistLimitDegrees=2}
 {bone1Name=dyn_hatLeft2, bone2Name=dyn_hatLeft1, swing1LimitDegrees=5, swing2LimitDegrees=5, twistLimitDegrees=2}
 {bone1Name=dyn_hatBack1, bone2Name=dyn_mstr_hat, swing1LimitDegrees=5, swing2LimitDegrees=5, twistLimitDegrees=2}
 {bone1Name=dyn_hatBack2, bone2Name=dyn_hatBack1, swing1LimitDegrees=5, swing2LimitDegrees=5, twistLimitDegrees=2}
 {bone1Name=dyn_hatRight1, bone2Name=dyn_mstr_hat, swing1LimitDegrees=5, swing2LimitDegrees=5, twistLimitDegrees=2}
 {bone1Name=dyn_hatRight2, bone2Name=dyn_hatRight1, swing1LimitDegrees=5, swing2LimitDegrees=5, twistLimitDegrees=2}
]
